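# Translucent lit Niagara mesh-particle material for a motion-blur effect on the Sunrise Castle male glider, using its diffuse, normal and mask textures.
# A second-UV streak texture drives opacity, scaled by vertex-color blue, and the same UV set defines a wing mask.
# UV-driven mip bias softens the glider textures.
# A previous-frame-only Z offset, reduced by a front-end particle parameter, fakes velocity alongside translucent velocity output.
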Material /BRCosmetics/Gadgets/Assets/VinderTech_GliderChute/Glider_Sunrise_Castle_Male/FX/M_FX_SunriseCastleGliderBlur_01
graph {
 properties {
  allowTranslucentCustomDepthWrites=true
  bOutputTranslucentVelocity=true
  bUsedWithNiagaraMeshParticles=true
  blendMode=BLEND_Translucent
  opacityMaskClipValue=1
  pixelDepthOffsetMode=PDOM_Legacy
  translucencyLightingMode=TLM_VolumetricDirectional
  usedShadingModels=Default Lit
 }
 nodes [
  n1 = TextureCoordinate() {coordinateIndex=1}
  n2 = TextureSample(UVs=@n1) {
   samplerType=SAMPLERTYPE_LinearColor
   texture=&/T_FX_SunriseCastleMale_MotionBlurStreaks
  }
  n3 = ScalarParameter() {defaultValue=0.3905, parameterName=Opacity_BlurEdge}
  n4 = ScalarParameter() {parameterName=Opacity_Blackened}
  n5 = VertexColor()
  LinearInterpolate(A=Constant(), B=Multiply(A=@n2.G, B=LinearInterpolate(A=@n3, B=@n4, Alpha=@n5.B)), Alpha=NamedRerouteUsage()) {constA=1}
  n6 = TextureCoordinate() {coordinateIndex=1}
  NamedRerouteDeclaration(None=Step(Y=Constant(), X=ComponentMask(None=@n6) {r=true})) {name=IsWing, nodeColor=<rgba 0 1 0.9766 1>}
  PreviousFrameSwitch(
   Current Frame=Constant3Vector()
   Previous Frame=LinearInterpolate(
    A=Constant3Vector() {constant=<rgba 0 0 111 0>}
    B=Constant3Vector() {constant=<rgba 0 0 10 0>}
    Alpha=DynamicParameter() {paramNames [Param1, Param2, IsFrontEnd, ""]}.IsFrontEnd
   )
  )
  n7 = ScalarParameter() {parameterName=MipA}
  n8 = ScalarParameter() {parameterName=MipB}
  n9 = Abs(None=ComponentMask(None=@n1) {r=true})
  n10 = TextureSample!(
   UVs
   Tex
   Bias=NamedRerouteDeclaration(None=LinearInterpolate(A=@n7, B=@n8, Alpha=@n9)) {name=MipLevel, nodeColor=<rgba 1 0 0.5625 1>}
   Apply View MipBias
  ) {
   mipValueMode=TMVM_MipBias
   texture=/BRCosmetics/Gadgets/Assets/VinderTech_GliderChute/Glider_Sunrise_Castle_Male/Textures/T_Glider_Sunrise_Castle_Male_D
  }
  Reroute(None=@n10.RGB)
  n11 = TextureSample(Bias=NamedRerouteUsage()) {
   mipValueMode=TMVM_MipBias
   samplerType=SAMPLERTYPE_Normal
   texture=/BRCosmetics/Gadgets/Assets/VinderTech_GliderChute/Glider_Sunrise_Castle_Male/Textures/T_Glider_Sunrise_Castle_Male_N
  }
  n12 = Transform(None=Constant3Vector() {constant=<rgba 0 0 1 0>}) {transformSourceType=TRANSFORMSOURCE_World, transformType=TRANSFORM_Tangent}
  StaticSwitchParameter(True=@n11.RGB, False=LinearInterpolate(A=@n11.RGB, B=VertexInterpolator(VS=@n12).PS, Alpha=NamedRerouteUsage())) {parameterName=NormalIsStandardMap}
  TextureSample(Bias=NamedRerouteUsage()) {
   mipValueMode=TMVM_MipBias
   samplerType=SAMPLERTYPE_Masks
   texture=/BRCosmetics/Gadgets/Assets/VinderTech_GliderChute/Glider_Sunrise_Castle_Male/Textures/T_Glider_Sunrise_Castle_Male_S
  }
 ]
}
