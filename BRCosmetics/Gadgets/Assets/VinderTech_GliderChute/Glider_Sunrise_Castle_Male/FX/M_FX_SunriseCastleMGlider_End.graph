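# Unlit, two-sided translucent Niagara sprite and mesh-particle material for the Sunrise Castle male glider's "End" effect.
# A fast sine oscillates the texture's rotation around an off-center pivot, making it wobble.
# Red channel times particle color passes through the shared emissive-adjustment function.
# Green channel, faded along U, forms an opacity-style mask, with a scalar parameter blending in particle alpha.
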Material /BRCosmetics/Gadgets/Assets/VinderTech_GliderChute/Glider_Sunrise_Castle_Male/FX/M_FX_SunriseCastleMGlider_End
graph {
 properties {
  bUsedWithNiagaraMeshParticles=true
  bUsedWithNiagaraSprites=true
  blendMode=BLEND_Translucent
  pixelDepthOffsetMode=PDOM_Legacy
  shadingModel=MSM_Unlit
  twoSided=true
  usedShadingModels=Unlit
 }
 nodes [
  n1 = Sine(None=Time()) {period=0.1}
  n2 = TextureSample(UVs=Rotator(Coordinate=TextureCoordinate(), Time=Multiply(A=@n1) {constB=1.5}) {centerX=0.1}) {
   samplerSource=SSM_Clamp_WorldGroupSettings
   samplerType=SAMPLERTYPE_LinearColor
   texture=&/T_FX_SunriseCastleMGlider_End
  }
  n3 = TextureCoordinate()
  n4 = Multiply(A=Multiply(A=@n2.G) {constB=1.2}, B=OneMinus(None=ComponentMask(None=@n3) {r=true})) {constB=0.5}
  n5 = ParticleColor()
  LinearInterpolate(A=@n4, B=Multiply(A=@n5.A, B=@n4), Alpha=ScalarParameter() {parameterName=UseAlphaParticleColor})
  MaterialFunctionCall(Emissive=Multiply(A=@n5.RGB, B=@n2.R)) {materialFunction=&MF_FX_EmissiveAdjustments}
 ]
}
