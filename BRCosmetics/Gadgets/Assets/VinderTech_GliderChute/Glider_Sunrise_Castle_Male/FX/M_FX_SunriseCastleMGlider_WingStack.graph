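# Masked, two-sided lit material for Niagara mesh particles of the Sunrise Castle Male glider wing stack, using the glider's diffuse, normal and mask textures.
# A sine over time rotates vertices around a local axis at the object pivot, producing a flapping motion.
# Dynamic parameter 2 offsets the phase, and vertex-color blue, shaped by a power, adds lag across the mesh.
# That blue channel times OpacityAmount also feeds DitherTemporalAA for dithered opacity.
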Material /BRCosmetics/Gadgets/Assets/VinderTech_GliderChute/Glider_Sunrise_Castle_Male/FX/M_FX_SunriseCastleMGlider_WingStack
graph {
 properties {bUsedWithNiagaraMeshParticles=true, blendMode=BLEND_Masked, pixelDepthOffsetMode=PDOM_Legacy, twoSided=true, usedShadingModels=Default Lit}
 nodes [
  MaterialFunctionCall(Alpha Threshold=Multiply(A=VertexColor().B, B=ScalarParameter() {parameterName=OpacityAmount})) {materialFunction=&DitherTemporalAA}
  RotateAboutAxis()
  n1 = VectorParameter() {parameterName=Axis, defaultValue=<rgba 1 0 0 1>}
  n2 = Time()
  n3 = ScalarParameter() {parameterName=Speed}
  n4 = VertexColor()
  n5 = ScalarParameter() {parameterName=LagShape}
  n6 = Add(
   A=Add(A=DynamicParameter().Param2, B=Multiply(A=@n2, B=@n3))
   B=Multiply(A=Power(Base=@n4.B, Exp=@n5), B=ScalarParameter() {parameterName=LagAmount})
  )
  RotateAboutAxis(
   NormalizedRotationAxis=Transform(None=Normalize(VectorInput=@n1.RGB)) {transformSourceType=TRANSFORMSOURCE_Local}
   RotationAngle=Multiply(A=Sine(None=@n6) {period=0.2}, B=ScalarParameter() {parameterName=Angle}) {constB=0.05}
   PivotPoint=MaterialFunctionCall() {materialFunction=&ObjectPivotPoint}.Object Pivot Location
   Position=WorldPosition().XYZ
  )
  TextureSample!(UVs, Tex, Bias, Apply View MipBias) {
   mipValueMode=TMVM_MipBias
   samplerType=SAMPLERTYPE_Normal
   texture=/BRCosmetics/Gadgets/Assets/VinderTech_GliderChute/Glider_Sunrise_Castle_Male/Textures/T_Glider_Sunrise_Castle_Male_N
  }
  TextureSample() {
   mipValueMode=TMVM_MipBias
   samplerType=SAMPLERTYPE_Masks
   texture=/BRCosmetics/Gadgets/Assets/VinderTech_GliderChute/Glider_Sunrise_Castle_Male/Textures/T_Glider_Sunrise_Castle_Male_S
  }
  TextureSample() {
   mipValueMode=TMVM_MipBias
   texture=/BRCosmetics/Gadgets/Assets/VinderTech_GliderChute/Glider_Sunrise_Castle_Male/Textures/T_Glider_Sunrise_Castle_Male_D
  }
 ]
}
