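# Translucent lit Niagara mesh-particle material that fakes motion blur on the Sunrise Castle male glider.
# Two streak texture samples pan on sine waves in vertex interpolators, then combine with particle alpha.
# A previous-frame-only offset, weighted by dynamic parameters with a front-end variant, creates velocity for engine blur at Epic and default quality.
# A camera offset adds to it, alongside a UV-based blur mask and glider diffuse and normal textures.
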Material /BRCosmetics/Gadgets/Assets/VinderTech_GliderChute/Glider_Sunrise_Castle_Male/FX/M_FX_SunriseCastleMaleGlider_Blur
graph {
 properties {
  bOutputTranslucentVelocity=true
  bUsedWithNiagaraMeshParticles=true
  blendMode=BLEND_Translucent
  opacityMaskClipValue=0
  pixelDepthOffsetMode=PDOM_Legacy
  translucencyLightingMode=TLM_VolumetricPerVertexDirectional
  usedShadingModels=Default Lit
 }
 nodes [
  Constant3Vector() {constant=<rgba 0 0 1 0>}
  n1 = ScalarParameter() {parameterName=Panning_Ctr}
  n2 = MaterialFunctionCall(Value=Multiply(A=NamedRerouteUsage(), B=ScalarParameter() {parameterName=Speed_1}) {constB=3}) {materialFunction=&LinearSine}
  n3 = VertexInterpolator(VS=AppendVector(A=Constant(), B=Add(A=@n1, B=@n2.Linear Sine)))
  n4 = TextureCoordinate() {coordinateIndex=1}
  n5 = TextureSample(UVs=MaterialFunctionCall(UVs=Add(A=@n3.PS, B=@n4), Texture Scale=ScalarParameter() {defaultValue=1, parameterName=Panning_Scale}) {materialFunction=&ScaleUVsByCenter}.UVs) {
   samplerSource=SSM_Clamp_WorldGroupSettings
   samplerType=SAMPLERTYPE_LinearColor
   texture=&/T_FX_SunriseCastleMale_MotionBlurStreaks
  }
  n6 = ScalarParameter() {parameterName=Panning_Ctr}
  n7 = MaterialFunctionCall(Value=Multiply(A=NamedRerouteUsage(), B=ScalarParameter() {parameterName=Speed_2}) {constB=3}) {materialFunction=&LinearSine}
  n8 = VertexInterpolator(VS=AppendVector(A=Constant(), B=Add(A=@n6, B=@n7.Linear Sine)))
  n9 = TextureCoordinate() {coordinateIndex=1}
  n10 = TextureSample(UVs=MaterialFunctionCall(UVs=Add(A=@n8.PS, B=@n9), Texture Scale=ScalarParameter() {defaultValue=1, parameterName=Panning_Scale}) {materialFunction=&ScaleUVsByCenter}.UVs) {
   samplerSource=SSM_Clamp_WorldGroupSettings
   samplerType=SAMPLERTYPE_LinearColor
   texture=&T_FX_SunriseCastleMale_MotionBlurStreaks
  }
  n11 = Add(A=LinearInterpolate(A=@n5.G, B=@n5.G, Alpha=@n2.Direction), B=LinearInterpolate(A=@n10.G, B=@n10.G, Alpha=@n7.Direction))
  Multiply(A=Multiply(A=NamedRerouteUsage(), B=Multiply(A=@n11)), B=ParticleColor().A)
  n12 = ComponentMask(None=TextureCoordinate() {coordinateIndex=1}) {r=true}
  n13 = Constant()
  NamedRerouteDeclaration(None=OneMinus(None=SphereMask(A=@n12, B=@n13) {attenuationRadius=0.5, hardnessPercent=0})) {name=BlurMask, nodeColor=<rgba 0 0.0156 1 1>}
  n14 = DynamicParameter() {defaultValue=<rgba 0 1 1 1>, paramNames [MotionBlurWeight, RandomTime, IsFrontEnd, Param4]}
  NamedRerouteDeclaration(None=Add(A=@n14.RandomTime, B=Time())) {name=OffsetTime, nodeColor=<rgba 0.5469 1 0 1>}
  n15 = Transform(None=LinearInterpolate(
    A=Constant3Vector() {constant=<rgba 0 0 40 0>}
    B=VectorParameter() {parameterName=FrontEndMotionBlurAmount, defaultValue=<rgba 0 0 80 1>}.RGB
    Alpha=DynamicParameter() {defaultValue=<rgba 0 1 1 1>, paramNames [MotionBlurWeight, RandomTime, IsFrontEnd, Param4]}.IsFrontEnd
   )) {transformSourceType=TRANSFORMSOURCE_ParticleWorld}
  n16 = NamedRerouteUsage()
  n17 = PreviousFrameSwitch(Current Frame=Constant3Vector(), Previous Frame=Multiply(A=Multiply(A=@n15, B=@n16), B=Clamp(None=@n14.MotionBlurWeight)))
  n18 = Constant3Vector()
  n19 = Multiply(A=ScalarParameter() {parameterName=CamOffset_WingTip}, B=NamedRerouteUsage())
  Reroute(None=Add(
    A=QualitySwitch(Default=@n17, Low=@n18, High=@n18, Medium=@n18, Epic=@n17)
    B=MaterialFunctionCall(Offset Amount=@n19) {materialFunction=/Engine/Functions/Engine_MaterialFunctions02/WorldPositionOffset/CameraOffset}.Camera Offset
   ))
  TextureSample() {texture=/BRCosmetics/Gadgets/Assets/VinderTech_GliderChute/Glider_Sunrise_Castle_Male/Textures/T_Glider_Sunrise_Castle_Male_D}
  TextureSample() {
   samplerType=SAMPLERTYPE_Normal
   texture=/BRCosmetics/Gadgets/Assets/VinderTech_GliderChute/Glider_Sunrise_Castle_Male/Textures/T_Glider_Sunrise_Castle_Male_N
  }
 ]
}
