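# Idle Niagara effect for the Sunrise Castle Male glider, drawing wing trails and motion-blur visuals on its skeletal mesh.
# FN_AnimTrail emits two short-lived bone-pair ribbons, varying color by time of day and width, velocity and alpha by front-end state.
# MotionBlurMeshes001 pins six persistent mesh/sprite particles by WingIndex, mirroring them per side and driving Z scale and material from FX_MotionBlurWeight.
# GPU speed lines and the original MotionBlurMeshes emitter are disabled.
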NiagaraSystem /BRCosmetics/Gadgets/Assets/VinderTech_GliderChute/Glider_Sunrise_Castle_Male/FX/NS_GliderSunriseCastleMale_Idle
dependencies {
 usedDataInterfaces [&NiagaraDataInterfaceCurve, &NiagaraDataInterfaceVector2DCurve]
 usedDynamicInputs [
  &RemapRange
  &VectorLength
  &Subtract_Float
  &DotProduct
  &NormalizeVector
  &Multiply_Vector
 ]
 usedModules [
  &SystemState
  &self:SystemSpawnScript.SystemScriptSource.SystemScriptGraph.NiagaraNodeAssignment_0.SetVariables_#1
  &self:SystemSpawnScript.SystemScriptSource.SystemScriptGraph.NiagaraNodeAssignment_1.SetVariables_#2
  &FN_Integrator
 ]
}
emitters {
 FN_AnimTrail {
  emitterSpawnScript [&FN_RibbonIDOffsetUtility()]
  emitterUpdateScript [
   &EmitterState()
   &FN_SpawnAnimTrail(Number of Anim Trails=2)
  ]
  particleSpawnScript [
   &FN_InitializeAnimTrailBoneIndecies(First Bone or Socket Index=1, Second Bone or Socket Index=0)
   &FN_InitializeAnimTrailBoneIndecies(AnimTrail Selection=1, First Bone or Socket Index=3, Second Bone or Socket Index=2) {module=FN_InitializeAnimTrailBoneIndecies001}
   &FN_InitializeAnimTrail(
    AnimTrailWidth=&MakeCustomFloatFromBool(Bool=$User.bIsFrontEnd, True Float=1, False Float=0.5)
    Lifetime=0.15
   )
   &AddVelocity(Velocity=<0 400 0>) {enabled=false}
   &AddVelocity(Velocity=&FN_FrontEndVector3Modifier(
     Gameplay=<0 1250 0>
     Front End=&Multiply_VectorByFloat(
      Vector=&MakeVector(
       X=0
       Y=1428.8495
       Z=&Multiply_Float(A=&Sine(Normalized Angle=$Emitter.Age, Period=0.7, Scale=1, Bias=5.7178), B=60)
      )
      Float=1
     )
    )) {module=AddVelocity001}
  ]
  particleUpdateScript [
   &InheritSourceMovement(
    Applied Position Scale=&VectorFromFloat(Value=&FloatFromCurve(FloatCurve=DataInterface() {propertyValues {Curve {keys [<at 0 1 RCIM_Cubic>,<at 1 0 RCIM_Cubic>]}}}))
    ApplyForce=false
    ApplyPosition=true
    ApplyPositionFalloff=false
   ) {enabled=false}
   &SolveForcesAndVelocity(Clamp Velocity=false, Limit Acceleration=false)
   &FN_UpdateAnimTrail(Locked Particle RibbonWidth=1, SkeletalMesh=$User.SkeletalMesh) {module=FN_UpdateAnimTrail001}
   &Color(
    Color=&MultiplyLinearColor(
     Linear Color=&MakeCustomLinearColorFromBool(Bool=$User.bIsFrontEnd, True Color=<rgba 0.5 0.5 0.5 0.6>, False Color=<rgba 1 1 1 0.2>)
     Scale Factor=&MakeFloat4FromVectorAndFloat(
      "Vector (RGB)"=&FN_Select_Vector_from_Time_Of_Day(Morning=<1 0.3 0.2>, Day=<1.5 1.5 1.5>, Evening=<0.7 0.7 0.7>, Night=<0.9 0.9 1>)
      "Float (Alpha)"=1
     )
    )
    Scale Alpha=&MakeCustomFloatFromBool(Bool=$User.bIsFrontEnd, True Float=1, False Float=$System.FN_Integrator001.Value)
   )
   /SetVariables(Particles.RibbonU0Override=$Particles.NormalizedAge) {isSetParameters=true}
  ]
  rendererClasses [&NiagaraRibbonRendererProperties]
  renderers [{rendererClass=&NiagaraRibbonRendererProperties}]
 }
 FN_SpeedLines {
  bEnabled=false
  simTarget=GPUComputeSim
  emitterUpdateScript [
   &EmitterState()
   &SpawnRate(SpawnRate=&Multiply_Float(A=250, B=$System.FN_Integrator001.Value), Use Spawn Probability=false)
  ]
  particleSpawnScript [
   /Niagara/Modules/Spawn/Initialization/InitializeParticle(
    Lifetime=&UniformRangedFloat(Minimum=0.3, Maximum=0.6)
    Module.Write Color=true
    Module.Write Lifetime=true
    Module.Write Mass=true
    Module.Write Material Random=true
    Module.Write Position=false
    Module.Write Scale=false
    Module.Write SpriteRotation=true
    Module.Write SpriteSize=true
    Position=&SimulationPosition()
    Sprite Rotation=&UniformRangedFloat(Minimum=0, Maximum=360)
    Sprite Size=&UniformRangedVector2D(Minimum=<3 1>, Maximum=<1.5 1>)
   )
   &TorusLocation(
    Align Torus With Axis=true
    Coordinate Space:ENiagaraCoordinateSpace=::Simulation
    Large Radius=150
    MaskSpawnGroup=false
    Module.Override Randomness=false
    Offset=<0 -200 0>
    Torus Axis=/Game/Effects/Niagara/DynamicInputs/Engine/FN_GetSystemAxis(Axis:ENiagaraOrientationAxis=::YAxis, Flip Axis:NiagaraBool=false)
    Torus Origin:NiagaraPosition=/Niagara/DynamicInputs/Helpers/SimulationPosition()
   ) {enabled=false}
   &InheritVelocity(Module.Limit Speed=true, Velocity Scale=<0.7 0.7 0>) {enabled=false}
   &AddVelocity(Velocity=<0 600 0>)
   &BoxLocation(Box Offset=<0 -150 50>, Box Size=<600 100 100>, MaskSpawnGroup=false, Module.Override Seed=false, Rotate Box=false, Surface Only=false)
  ]
  particleUpdateScript [
   &ParticleState()
   &Color(
    Color=<rgba 1 1 1 0.1>
    Scale Alpha=&FloatFromCurve(FloatCurve=DataInterface() {propertyValues {Curve {keys [<at 0 0 RCIM_Cubic>,<at 0.5 1 RCIM_Cubic>,<at 1 0 RCIM_Cubic>]}}})
   )
   &SolveForcesAndVelocity(Clamp Velocity=false, Limit Acceleration=false)
   &ScaleSpriteSize(Scale Factor=&Vector2DFromCurve(
     Vector2Curve=DataInterface() {propertyValues {
       XCurve {keys [<at -0.0069 0.5 RCIM_Cubic>,<at 0.4918 1 RCIM_Cubic>,<at 1 0.5 RCIM_Cubic>]}
       YCurve {keys [<at 0 0.0245 RCIM_Cubic>, {arriveTangent=-0.0245, interpMode=RCIM_Cubic, leaveTangent=-0.0245, time=0.4975, value=1}, <at 1 0 RCIM_Cubic>]}
      }}
     ScaleCurve=<1 1>
    ))
   &ScaleSpriteSizeBySpeed(Max Scale Factor=<2 80>)
  ]
  rendererClasses [&NiagaraSpriteRendererProperties]
  renderers [{rendererClass=&NiagaraSpriteRendererProperties}]
 }
 MotionBlurMeshes {
  bEnabled=false
  emitterUpdateScript [
   &EmitterState(
    Inactive Response=::NewEnumerator1
    Life Cycle Mode=::NewEnumerator1
    Loop Behavior=::NewEnumerator1
    Loop Duration Mode=::NewEnumerator1
    UseLoopDelay=false
   )
   &SpawnBurst_Instantaneous(Spawn Count=6, Use Spawn Probability=false, UseLoopCountLimit=false)
  ]
  particleSpawnScript [
   /Niagara/Modules/Spawn/Initialization/V2/InitializeParticle(Color=<rgba 1 1 1 0.8>, Lifetime=0, Mesh Scale Mode=::NewEnumerator3, UsePositionOffset=false, Write Material Random=true)
   /SetVariables(Particles.WingIndex:NiagaraInt32=&ModuloInt(A=$Particles.UniqueID, B=6)) {isSetParameters=true, module=SetVariables_#3}
   /Niagara/Modules/Update/Orientation/V2/OrientMeshToVector(
    Look At Direction:Vector3f=&MakeCustomVectorFromBool(
     Bool=&SetBoolByIntComparison(A=$Particles.WingIndex, B=2, Comparison Type=::NewEnumerator0)
     True Vector=<-1 0 0>
     False Vector=<1 0 0>
    )
    Override Quat=false
    Reference Axis Vector:Vector3f=<0 0 0>
    Side Axis Direction:Vector3f=&MakeCustomVectorFromBool(
     Bool=&SetBoolByIntComparison(A=$Particles.WingIndex, B=2, Comparison Type=::NewEnumerator0)
     True Vector=<0 -1 0>
     False Vector=<0 1 0>
    )
    Up Axis Direction:Vector3f=<0 0 0>
    UseOptionalSideAxis=true
    UseOptionalUpAxis=false
   )
  ]
  particleUpdateScript [
   &ParticleState(Kill Particles When Lifetime Has Elapsed=false, Let Infinitely Lived Particles Die When Emitter Deactivates=true)
   &SkeletalMeshLocation(Bone / Socket Index=$Particles.WingIndex, Bone Sampling Mode=::NewEnumerator4, MaskSpawnGroup=false, Skeletal Mesh=$User.SkeletalMesh)
   /SetVariables(Particles.MeshOrientation=&MultiplyQuaternion(
     Quaternion A=&MultiplyQuaternion(
      Quaternion A=$StackContext.MeshOrientation
      Quaternion B=/Niagara/DynamicInputs/TypeConversions/MakeCustomQuaternionFromBool(
       Bool:NiagaraBool=&SetBoolByIntComparison(A=$Particles.WingIndex, B=2, Comparison Type=::NewEnumerator0)
       True Quat:Quat4f=&MultiplyQuaternion(
        Quaternion A=&MakeQuatFromAxisAngle(Axis=<0 0 1>, AngleInDegrees=90, NormalizeAxis=true)
        Quaternion B=&MakeQuatFromAxisAngle(Axis=<1 0 0>, AngleInDegrees=180)
       )
       False Quat:Quat4f=&MakeQuatFromAxisAngle(Axis=<0 0 1>, AngleInDegrees=270, NormalizeAxis=true)
      )
     )
     Quaternion B=&MakeQuatFromAxisAngle(Axis=<1 0 0>, AngleInDegrees=270)
    )) {isSetParameters=true, module=SetVariables_#4}
   /SetVariables(Particles.MeshOrientation=&MultiplyQuaternion(
     Quaternion A=$StackContext.MeshOrientation
     Quaternion B=&MakeQuatFromAxisAngle(
      Axis=<0 1 0>
      AngleInDegrees=&Multiply_Float(A=&Sine(Normalized Angle=$Emitter.Age, Period=0.1, Scale=1, Bias=0), B=3)
     )
    )) {enabled=false, isSetParameters=true, module=SetVariables_#5}
   &DynamicMaterialParameters(
    Index 0 Param 1=$System.FX_MotionBlurWeight
    Index 0 Param 2=&RandomRangeFloat(Minimum=0, Maximum=3)
    Index 0 Param 3=&MakeCustomFloatFromBool(Bool=$User.bIsFrontEnd, True Float=1, False Float=0)
    Module.Param0WriteEnabled=true
    Module.Param1WriteEnabled=true
    Module.Param2WriteEnabled=true
    Module.Param3WriteEnabled=true
   )
   &ScaleMeshSize(Scale Factor=&MakeVector(X=1, Y=1, Z=$System.FX_MotionBlurWeight))
   &ScaleColor(Scale Alpha=&ClampFloat(Float=$System.FX_MotionBlurWeight), ScaleA=true, ScaleRGB=true)
  ]
  rendererClasses [&NiagaraMeshRendererProperties]
  renderers [
   {rendererClass=&NiagaraMeshRendererProperties, rendererIndex=0}
   {rendererClass=&NiagaraMeshRendererProperties, rendererIndex=1}
   {rendererClass=&NiagaraMeshRendererProperties, rendererIndex=2}
   {rendererClass=&NiagaraMeshRendererProperties, rendererIndex=3}
  ]
 }
 MotionBlurMeshes001 {
  emitterUpdateScript [
   &EmitterState(
    Inactive Response=::NewEnumerator1
    Life Cycle Mode=::NewEnumerator1
    Loop Behavior=::NewEnumerator1
    Loop Duration Mode=::NewEnumerator1
    Scalability Mode=::NewEnumerator1
    UseLoopDelay=false
   )
   &SpawnBurst_Instantaneous(Spawn Count=6, Use Spawn Probability=false, UseLoopCountLimit=false)
  ]
  particleSpawnScript [
   &InitializeParticle(Color=<rgba 0.1771 0.1483 0.1226 0.85>, Lifetime=0, Mesh Scale Mode=::NewEnumerator3, UsePositionOffset=false, Write Material Random=true)
   /SetVariables(Particles.WingIndex:NiagaraInt32=&ModuloInt(A=$Particles.UniqueID, B=6)) {isSetParameters=true, module=SetVariables_#3}
   &OrientMeshToVector(
    Look At Direction:Vector3f=&MakeCustomVectorFromBool(
     Bool=&SetBoolByIntComparison(A=$Particles.WingIndex, B=2, Comparison Type=::NewEnumerator0)
     True Vector=<-1 0 0>
     False Vector=<1 0 0>
    )
    Override Quat=false
    Reference Axis Vector:Vector3f=<0 0 0>
    Side Axis Direction:Vector3f=&MakeCustomVectorFromBool(
     Bool=&SetBoolByIntComparison(A=$Particles.WingIndex, B=2, Comparison Type=::NewEnumerator0)
     True Vector=<0 -1 0>
     False Vector=<0 1 0>
    )
    Up Axis Direction:Vector3f=<0 0 0>
    UseOptionalSideAxis=true
    UseOptionalUpAxis=false
   )
  ]
  particleUpdateScript [
   &ParticleState(Kill Particles When Lifetime Has Elapsed=false, Let Infinitely Lived Particles Die When Emitter Deactivates=true)
   &SkeletalMeshLocation(Bone / Socket Index=$Particles.WingIndex, Bone Sampling Mode=::NewEnumerator4, MaskSpawnGroup=false, Skeletal Mesh=$User.SkeletalMesh)
   /SetVariables(Particles.MeshOrientation=&MultiplyQuaternion(
     Quaternion A=&MultiplyQuaternion(
      Quaternion A=$StackContext.MeshOrientation
      Quaternion B=&MakeCustomQuaternionFromBool(
       Bool:NiagaraBool=&SetBoolByIntComparison(A=$Particles.WingIndex, B=2, Comparison Type=::NewEnumerator0)
       True Quat:Quat4f=&MultiplyQuaternion(
        Quaternion A=&MakeQuatFromAxisAngle(Axis=<0 0 1>, AngleInDegrees=85, NormalizeAxis=true)
        Quaternion B=&MakeQuatFromAxisAngle(Axis=<1 0 0>, AngleInDegrees=0)
       )
       False Quat:Quat4f=&MakeQuatFromAxisAngle(Axis=<0 0 1>, AngleInDegrees=270, NormalizeAxis=true)
      )
     )
     Quaternion B=&MakeQuatFromAxisAngle(Axis=<1 0 0>, AngleInDegrees=270)
    )) {isSetParameters=true, module=SetVariables_#4}
   /SetVariables(Particles.MeshOrientation=&MultiplyQuaternion(
     Quaternion A=$StackContext.MeshOrientation
     Quaternion B=&MakeQuatFromAxisAngle(Axis=<0 -1 0>, AngleInDegrees=90)
    )) {enabled=false, isSetParameters=true}
   /SetVariables(Particles.MeshOrientation=&MultiplyQuaternion(
     Quaternion A=$StackContext.MeshOrientation
     Quaternion B=&MakeQuatFromAxisAngle(
      Axis=<1 0 0>
      AngleInDegrees=&Multiply_Float(A=&Sine(Normalized Angle=$Emitter.Age, Period=0.2, Scale=2, Bias=0), B=0)
     )
    )) {isSetParameters=true, module=SetVariables_#5}
   &DynamicMaterialParameters(
    Index 0 Param 1=$System.FX_MotionBlurWeight
    Index 0 Param 2=&RandomRangeFloat(Minimum=0, Maximum=3)
    Index 0 Param 3=&MakeCustomFloatFromBool(Bool=$User.bIsFrontEnd, True Float=1, False Float=0)
    Index 0 Param 4=1
    Module.Param0WriteEnabled=true
    Module.Param1WriteEnabled=true
    Module.Param2WriteEnabled=true
    Module.Param3WriteEnabled=true
   )
   &ScaleMeshSize(Scale Factor=&MakeVector(X=1, Y=1, Z=$System.FX_MotionBlurWeight))
   &ScaleColor(ScaleA=true, ScaleRGB=true)
   /SetVariables() {isSetParameters=true}
   /SetVariables(Particles.SpriteAlignment=&MakeCustomVectorFromBool(
     Bool=&SetBoolByIntComparison(A=$Particles.WingIndex, B=2, Comparison Type=::NewEnumerator0)
     True Vector=<1 0 0>
     False Vector=<-1 0 0>
    )) {isSetParameters=true}
   /SetVariables(Particles.SpriteSize=&Multiply_Vector2D(
     A=<57.6782 212.4045>
     B=&MakeCustomVector2DFromBool(
      Bool=&SetBoolByIntComparison(A=$Particles.WingIndex, B=2, Comparison Type=::NewEnumerator0)
      True Vector2D=<-1 1>
      False Vector2D=<1 1>
     )
    )) {isSetParameters=true}
   /SetVariables(Particles.PivotOffset:Vector2f=&MakeCustomVector2DFromBool(
     Bool=&SetBoolByIntComparison(A=$Particles.WingIndex, B=2, Comparison Type=::NewEnumerator0)
     True Vector2D=<0.5 1.2>
     False Vector2D=<0.5 1.2>
    )) {isSetParameters=true}
  ]
  rendererClasses [&NiagaraMeshRendererProperties, &NiagaraSpriteRendererProperties]
  renderers [
   {rendererClass=&NiagaraMeshRendererProperties, rendererIndex=0}
   {rendererClass=&NiagaraMeshRendererProperties, rendererIndex=1}
   {rendererClass=&NiagaraSpriteRendererProperties, rendererIndex=2}
   {rendererClass=&NiagaraMeshRendererProperties, rendererIndex=3}
   {rendererClass=&NiagaraMeshRendererProperties, rendererIndex=4}
  ]
 }
}
summary {
 systemName=NS_GliderSunriseCastleMale_Idle
 emitters [
  {bEnabled=false, emitterName=MotionBlurMeshes, rendererClasses [&NiagaraMeshRendererProperties]}
  {emitterName=FN_AnimTrail, rendererClasses [&NiagaraRibbonRendererProperties]}
  {bEnabled=false, emitterName=FN_SpeedLines, simTarget=GPUComputeSim, rendererClasses [&NiagaraSpriteRendererProperties]}
  {
   emitterName=MotionBlurMeshes001
   rendererClasses [&NiagaraMeshRendererProperties, &NiagaraSpriteRendererProperties]
  }
 ]
 userVariables [
  User.bIsFrontEnd {type=NiagaraBool}
  User.bIsFrontEndPreview {type=NiagaraBool}
  User.bisFullyDeployed {type=NiagaraBool}
  User.ForwardDot {type=NiagaraFloat, defaultValue=0, description=""}
  User.RibbonIDOffset {type=NiagaraInt32}
  User.SkeletalMesh {
   type=NiagaraDataInterfaceSkeletalMesh
   defaultValue {
    struct=NiagaraExt_VariableValue_DataInterface
    value {dataInterface=&self:NiagaraDataInterfaceSkeletalMesh_0}
   }
  }
  User.Speed {type=NiagaraFloat, defaultValue=0, description=""}
 ]
}
